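# Physics asset for the PER_2 hair part of the medium male Elastic player body, defining capsule bodies for its dynamic hair and ponytail bones.
# A tiny MSTR capsule serves as the anchor, with the first hair bone's swing and twist locked to it.
# A second hair bone chains off the first and branches into five two-bone ponytail strands (A–E).
# The hair capsules share one shape, and the joints share uniform, small swing and twist angle settings.
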PhysicsAsset /Game/Characters/Player/Male/Medium/Bodies/M_MED_Elastic/Meshes/Parts/PER_Parts/M_MED_Elastic_PER_2_Physics
bodies {
 dyn_hairPER2_01 [{shapeType=Capsule, length=3.8522, radius=1.0798, center=<2.1511 0 0>, rotation=<rot -90 0 0>}]
 dyn_hairPER2_01_MSTR [{shapeType=Capsule, length=0.01, radius=0.01}]
 dyn_hairPER2_02 [{shapeType=Capsule, length=3.8522, radius=1.0798, center=<2.1511 0 0>, rotation=<rot -90 0 0>}]
 dyn_hairPER2_ponytailA_01 [{shapeType=Capsule, length=3.8522, radius=1.0798, center=<2.1511 0 0>, rotation=<rot -90 0 0>}]
 dyn_hairPER2_ponytailA_02 [{shapeType=Capsule, length=3.8522, radius=1.0798, center=<2.1511 0 0>, rotation=<rot -90 0 0>}]
 dyn_hairPER2_ponytailB_01 [{shapeType=Capsule, length=3.8522, radius=1.0798, center=<2.1511 0 0>, rotation=<rot -90 0 0>}]
 dyn_hairPER2_ponytailB_02 [{shapeType=Capsule, length=3.8522, radius=1.0798, center=<2.1511 0 0>, rotation=<rot -90 0 0>}]
 dyn_hairPER2_ponytailC_01 [{shapeType=Capsule, length=3.8522, radius=1.0798, center=<2.1511 0 0>, rotation=<rot -90 0 0>}]
 dyn_hairPER2_ponytailC_02 [{shapeType=Capsule, length=3.8522, radius=1.0798, center=<2.1511 0 0>, rotation=<rot -90 0 0>}]
 dyn_hairPER2_ponytailD_01 [{shapeType=Capsule, length=3.8522, radius=1.0798, center=<2.1511 0 0>, rotation=<rot -90 0 0>}]
 dyn_hairPER2_ponytailD_02 [{shapeType=Capsule, length=3.8522, radius=1.0798, center=<2.1511 0 0>, rotation=<rot -90 0 0>}]
 dyn_hairPER2_ponytailE_01 [{shapeType=Capsule, length=3.8522, radius=1.0798, center=<2.1511 0 0>, rotation=<rot -90 0 0>}]
 dyn_hairPER2_ponytailE_02 [{shapeType=Capsule, length=3.8522, radius=1.0798, center=<2.1511 0 0>, rotation=<rot -90 0 0>}]
}
constraints [
 {
  bone1Name=dyn_hairPER2_01
  bone2Name=dyn_hairPER2_01_MSTR
  swing1LimitDegrees=5
  swing1Motion=Locked
  swing2LimitDegrees=5
  swing2Motion=Locked
  twistLimitDegrees=5
  twistMotion=Locked
 }
 {bone1Name=dyn_hairPER2_02, bone2Name=dyn_hairPER2_01, swing1LimitDegrees=5, swing2LimitDegrees=5, twistLimitDegrees=5}
 {bone1Name=dyn_hairPER2_ponytailA_01, bone2Name=dyn_hairPER2_02, swing1LimitDegrees=5, swing2LimitDegrees=5, twistLimitDegrees=5}
 {bone1Name=dyn_hairPER2_ponytailB_01, bone2Name=dyn_hairPER2_02, swing1LimitDegrees=5, swing2LimitDegrees=5, twistLimitDegrees=5}
 {bone1Name=dyn_hairPER2_ponytailC_01, bone2Name=dyn_hairPER2_02, swing1LimitDegrees=5, swing2LimitDegrees=5, twistLimitDegrees=5}
 {bone1Name=dyn_hairPER2_ponytailD_01, bone2Name=dyn_hairPER2_02, swing1LimitDegrees=5, swing2LimitDegrees=5, twistLimitDegrees=5}
 {bone1Name=dyn_hairPER2_ponytailE_01, bone2Name=dyn_hairPER2_02, swing1LimitDegrees=5, swing2LimitDegrees=5, twistLimitDegrees=5}
 {bone1Name=dyn_hairPER2_ponytailA_02, bone2Name=dyn_hairPER2_ponytailA_01, swing1LimitDegrees=5, swing2LimitDegrees=5, twistLimitDegrees=5}
 {bone1Name=dyn_hairPER2_ponytailB_02, bone2Name=dyn_hairPER2_ponytailB_01, swing1LimitDegrees=5, swing2LimitDegrees=5, twistLimitDegrees=5}
 {bone1Name=dyn_hairPER2_ponytailC_02, bone2Name=dyn_hairPER2_ponytailC_01, swing1LimitDegrees=5, swing2LimitDegrees=5, twistLimitDegrees=5}
 {bone1Name=dyn_hairPER2_ponytailD_02, bone2Name=dyn_hairPER2_ponytailD_01, swing1LimitDegrees=5, swing2LimitDegrees=5, twistLimitDegrees=5}
 {bone1Name=dyn_hairPER2_ponytailE_02, bone2Name=dyn_hairPER2_ponytailE_01, swing1LimitDegrees=5, swing2LimitDegrees=5, twistLimitDegrees=5}
]
